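# Physics asset for the Banana Smoothie backpack accessory, defining secondary-motion bodies for its slush, two-segment straw and umbrella.
# A capsule on the pack bone is the anchor, while each Dyn_ bone carries a box body.
# Slush, umbrella and lower straw constrain directly to the pack, and the upper straw chains off the lower one.
# Tight swing and twist limits keep movement small, with the straw held stiffest.
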PhysicsAsset /Game/Accessories/FORT_Backpacks/M_MED_Banana_Smoothie/Meshes/M_MED_Banana_Smoothie_Pack_PhysicsAsset
bodies {
 Dyn_slush [{shapeType=Box, extentX=1, extentY=13.51, extentZ=13.51, center=<0 0 -2.8553>, rotation=<rot 90 0 0>}]
 Dyn_straw01 [{shapeType=Box, extentX=3, extentY=3, extentZ=2.5, center=<1 0.0001 0>, rotation=<rot 90 0 0>}]
 Dyn_straw02 [{shapeType=Box, extentX=4.5, extentY=5, extentZ=3, center=<1.587 0 0>, rotation=<rot 90 0 0>}]
 Dyn_umbrella [{shapeType=Box, extentX=3, extentY=3, extentZ=4.5, center=<2.2939 -0.0001 0>, rotation=<rot 90 0 0>}]
 pack [{shapeType=Capsule, length=3.4709, radius=4.9456, center=<0.0119 -5.8999 -6.4016>, rotation=<rot 4.1114 -89.9998 -0.0387>}]
}
constraints [
 {bone1Name=Dyn_straw01, bone2Name=pack, swing1LimitDegrees=10, swing2LimitDegrees=10, twistLimitDegrees=2}
 {bone1Name=Dyn_straw02, bone2Name=Dyn_straw01, swing1LimitDegrees=10, swing2LimitDegrees=10, twistLimitDegrees=2}
 {bone1Name=Dyn_umbrella, bone2Name=pack, swing1LimitDegrees=15, swing2LimitDegrees=15, twistLimitDegrees=15}
 {bone1Name=Dyn_slush, bone2Name=pack, swing1LimitDegrees=15, swing2LimitDegrees=15, twistLimitDegrees=5}
]
