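# Animation Blueprint for the Banana Smoothie backpack, adding physics motion and an animated face.
# The base pose copies the head mesh, falling back to a lobby idle, and layers the input pose from root.
# Three RigidBody setups switch on aiming and airborne states, then spine bend additively offsets the root.
# The event graph sets face material parameters from emote state and a blink curve.
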AnimBlueprint /Game/Accessories/FORT_Backpacks/M_MED_Banana_Smoothie/Meshes/M_MED_Banana_Smoothie_Pack_Skeleton_AnimBlueprint
target_skeleton=&/M_MED_Banana_Smoothie_Pack_Skeleton
graphs [
 AnimGraph {nodes [
   n1 = RigidBody(ComponentPose=Local To Component(LocalPose=Use cached pose 'InputPose'())) {
    bClampLinearTranslationLimitToRefPose=true
    bOverrideWorldGravity=true
    overridePhysicsAsset=&/M_MED_Banana_Smoothie_Pack_PhysicsAsset
    alphaBoolBlend {blendOption=Linear}
    componentAppliedLinearAccClamp=<2000 2000 2000>
    componentLinearAccScale=<1 1 1>
    componentLinearVelScale=<1 1 1>
    overrideWorldGravity {z=-980}
   }
   n2 = RigidBody(ComponentPose=Local To Component(LocalPose=Use cached pose 'InputPose'())) {
    bClampLinearTranslationLimitToRefPose=true
    bOverrideWorldGravity=true
    overridePhysicsAsset=&M_MED_Banana_Smoothie_Pack_PhysicsAsset
    alphaBoolBlend {blendOption=Linear}
    componentAppliedLinearAccClamp=<4000 4000 4000>
    componentLinearAccScale=<0.5 0.5 0.5>
    componentLinearVelScale=<0.5 0.5 0.5>
    overrideWorldGravity {z=-490}
   }
   n3 = Blend Poses by bool(
    BlendPose_0=Component To Local(ComponentPose=@n2)
    BlendPose_1=Component To Local(ComponentPose=@n1)
    bActiveValue=Get IsInAir() -> (IsInAir:bool)
   ) {blendType=Linear}
   n4 = RigidBody(ComponentPose=Local To Component(LocalPose=Use cached pose 'InputPose'())) {
    bClampLinearTranslationLimitToRefPose=true
    bOverrideWorldGravity=true
    overridePhysicsAsset=&M_MED_Banana_Smoothie_Pack_PhysicsAsset
    alphaBoolBlend {blendOption=Linear}
    componentAppliedLinearAccClamp=<3000 3000 3000>
    componentLinearAccScale=<0.334 0.334 0.334>
    componentLinearVelScale=<0.334 0.334 0.334>
    overrideWorldGravity {z=-980}
   }
   n5 = Component To Local(ComponentPose=@n4)
   n6 = Get IsADS() -> (IsADS:bool)
   n7 = "Transform (Modify) Bone - Bone: root"(
    ComponentPose=Local To Component(LocalPose=Blend Poses by bool(BlendPose_0=@n5, BlendPose_1=@n3, bActiveValue=@n6) {blendType=Linear})
    Translation=Get backpack_translation()
    Rotation=Get backpack_rotate()
   ) {
    rotationMode=BMM_Additive
    rotationSpace=BCS_BoneSpace
    translationMode=BMM_Additive
    translationSpace=BCS_BoneSpace
    alphaBoolBlend {blendOption=Linear}
    boneToModify {boneName=root}
   }
   Output Pose(Result=Component To Local(ComponentPose=@n7)) {layerGroup=None}
   n8 = Sequence Player 'Idle_Lobby'() {sequence=/Game/Animation/Game/Pets/Peely_Face/Backpack/Idle/Idle_Lobby}
   n9 = Use cached pose 'copypose_cache'()
   n10 = Get isValidHead() -> (isValidHead:bool)
   AnimGraphNode_SubInput_0 = Input Pose()
   Save cached pose 'InputPose'(Pose=Layered blend per bone(BasePose=Blend Poses by bool(BlendPose_0=@n9, BlendPose_1=@n8, bActiveValue=@n10) {blendType=Linear}, BlendPoses_0=@AnimGraphNode_SubInput_0) {layerSetup [{branchFilters [{boneName=root}]}]}) {cacheName=InputPose}
   Save cached pose 'copypose_cache'(Pose=Copy Pose From Mesh(SourceMeshComponent=Get head_skelmesh() -> (head_skelmesh:Skeletal Mesh Component Object Reference)) {bCopyCurves=true, bUseAttachedParent=true}) {cacheName=copypose_cache}
  ]}
 EventGraph {nodes [
   Event BlueprintUpdateAnimation()
   > n1 = Sequence!(execute) -> (then_0, then_1, then_2:Exec)
   > n2 = Branch(Condition="float >= float"(A=Get PawnVelocityZ(), B=1) {is=title K2Node_CallFunction})
   > Set IsInAir(IsInAir:bool=true) -> (Output_Get:bool)
   Set IsInAir(execute=@n2.else, IsInAir=false)
   n3 = Get Dynamic Material Instance(execute=@n1.then_1, SkeletalMeshComponent=GetPartSkeletalMeshComponent())
   > n4 = Branch(Condition=Get bIsPlayingEmote())
   > SetScalarParameterValue(self=@n3, ParameterName=Default Face, Value=6)
   n5 = GetCurveValue(CurveName=R_blink_pose)
   n6 = MapRangeClamped(Value=@n5, InRangeA=0, InRangeB=0.1)
   > SetScalarParameterValue(self=@n3, ParameterName=Hide Element 01, Value=@n6)
   > SetScalarParameterValue(self=@n3, ParameterName=Hide Element 02, Value=float - float(B=@n6) {is=title K2Node_CallFunction})
   > SetScalarParameterValue(self=@n3, ParameterName=Hide Element 03, Value=1)
   SetScalarParameterValue(execute=@n4.else, self=@n3, ParameterName=Default Face, Value=22)
   n7 = "To Float (Integer)"(InInt=Round(A=@n5) {is=title K2Node_CallFunction})
   > SetScalarParameterValue(self=@n3, ParameterName=Hide Element 03, Value=@n7)
   > SetScalarParameterValue(self=@n3, ParameterName=Hide Element 02, Value=float - float(B=@n7))
   > SetScalarParameterValue(self=@n3, ParameterName=Hide Element 01, Value=1)
   n8 = GetOwnerSkeletalMeshComponent()
   n9 = BreakRotator(InRot=BreakTransform(InTransform=GetDeltaTransformFromRefPose(self=@n8, BoneName=spine_05, BaseName=spine_01)).Rotation)
   Set backpack_rotate(execute=@n1.then_2, backpack_rotate=MakeRotator(Roll=MapRangeClamped(Value=@n9.Yaw, InRangeA=0, InRangeB=-30, OutRangeB=20)))
   > Set backpack_translation(backpack_translation=MakeVector(Y=MapRangeClamped(Value=@n9.Yaw, InRangeA=0, InRangeB=-30, OutRangeB=-3)))
   Event OnBeginTargeting() {is=title K2Node_Event}
   > Set IsADS(IsADS:bool=true) -> (Output_Get:bool)
   Event BlueprintInitializeAnimation()
   > Sequence!(execute) -> (then_0, then_1)
   n10 = TryGetPawnOwner()
   > n11 = Is Valid(InputObject=@n10)
   Set isValidHead(execute=[@n12.CastFailed, @n11.Is Not Valid], isValidHead:bool=false) -> (Output_Get:bool)
   n12 = Cast To FortPlayerPawnAthena(execute=@n11.Is Valid, Object=@n10)
   > Set head_skelmesh(head_skelmesh:Skeletal Mesh Component Object Reference=GetSkeletalMeshForPartType(self=@n12, PartType=Head)) -> (Output_Get:Skeletal Mesh Component Object Reference)
   > Set isValidHead(isValidHead=true)
   Event OnEndTargeting() {is=title K2Node_Event}
   > Set IsADS(IsADS=false)
  ]}
]
variables [
 backpack_rotate {type=Rotator}
 backpack_translation {type=Vector}
 head_skelmesh {type=SkeletalMeshComponent}
 IsADS {type=boolean}
 IsInAir {type=boolean}
 isValidHead {type=boolean}
]
